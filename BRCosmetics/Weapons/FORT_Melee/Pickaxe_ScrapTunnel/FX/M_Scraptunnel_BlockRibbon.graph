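# Masked, two-sided lit material for Niagara ribbons in the ScrapTunnel pickaxe effects, drawing scrolling cube blocks.
# Three panning samples of a packed cube texture are layered by blue-channel masks, yielding normals and tint blend, with dynamic parameters for offset and speed (third layer ignores TimeMult).
# Summed masks are eroded along ribbon length and edges, then multiplied by particle alpha.
# Time-of-day functions drive roughness and a faint sun-facing emissive.
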Material /BRCosmetics/Weapons/FORT_Melee/Pickaxe_ScrapTunnel/FX/M_Scraptunnel_BlockRibbon
graph {
 properties {
  bAutomaticallySetUsageInEditor=false
  bUsedWithNiagaraRibbons=true
  blendMode=BLEND_Masked
  opacityMaskClipValue=0.333
  translucencyLightingMode=TLM_Surface
  translucencyPass=MTP_BeforeDOF
  twoSided=true
  usedShadingModels=Default Lit
 }
 nodes [
  n1 = DynamicParameter() {paramNames [Offset,TimeMult,Param3,Param4]}
  n2 = Add(A=TextureCoordinate() {coordinateIndex=1}, B=@n1.Offset)
  n3 = VectorParameter() {parameterName=Texturecoord_1, defaultValue=<rgba 0.6 0.5 1 0>}
  n4 = ComponentMask(None=@n3.RGB) {g=true, r=true}
  n5 = Multiply(A=Time(), B=@n1.TimeMult)
  n6 = TextureSample(UVs=Panner(Coordinate=Multiply(A=@n2, B=@n4), Time=@n5, Speed=AppendVector(A=@n3.B, B=@n3.B))) {texture=&/T_ScrapTunnel_Cubes_Packed}
  n7 = ComponentMask(None=@n6.RGB) {g=true, r=true}
  n8 = VectorParameter() {parameterName=Layer Tints}
  n9 = Add(A=@n2, B=Constant2Vector() {g=0.25})
  n10 = VectorParameter() {parameterName=Texturecoord_2, defaultValue=<rgba 0.75 0.6 0.5 0>}
  n11 = ComponentMask(None=@n10.RGB) {g=true, r=true}
  n12 = TextureSample(UVs=Panner(Coordinate=Multiply(A=@n9, B=@n11), Time=@n5, Speed=AppendVector(A=@n10.B, B=@n10.B))) {texture=&T_ScrapTunnel_Cubes_Packed}
  n13 = ComponentMask(None=@n12.RGB) {g=true, r=true}
  n14 = Add(A=@n2, B=Constant2Vector() {g=0.44})
  n15 = VectorParameter() {parameterName=Texturecoord_3, defaultValue=<rgba 0.5 0.5 0.5 0>}
  n16 = ComponentMask(None=@n15.RGB) {g=true, r=true}
  n17 = TextureSample(UVs=Panner(Coordinate=Multiply(A=@n14, B=@n16), Speed=AppendVector(A=@n15.B, B=@n15.B))) {texture=&T_ScrapTunnel_Cubes_Packed}
  n18 = LinearInterpolate(
   A=LinearInterpolate(A=AppendVector(A=@n7, B=@n8.R), B=AppendVector(A=@n13, B=@n8.G), Alpha=Ceil(None=@n12.B))
   B=AppendVector(A=ComponentMask(None=@n17.RGB) {g=true, r=true}, B=@n8.B)
   Alpha=Ceil(None=@n17.B)
  )
  LinearInterpolate(
   A=VectorParameter() {defaultValue=<rgba 0.0208 0.0208 0.0208 1>}.RGB
   B=VectorParameter() {parameterName=Param_1, defaultValue=<rgba 0.1117 0.125 0.1023 1>}.RGB
   Alpha=ComponentMask(None=@n18) {b=true}
  )
  n19 = MaterialFunctionCall(
   Value_Morning=ScalarParameter() {defaultValue=0.5, parameterName=Rough_Morning}
   Value_Day=ScalarParameter() {defaultValue=0.45, parameterName=Rough_Day}
   Value_Evening=ScalarParameter() {defaultValue=0.5, parameterName=Rough_Evening}
   Value_Night=ScalarParameter() {defaultValue=0.5, parameterName=Rough_Night}
  ) {materialFunction=&Scalar_Time_of_Day}
  n20 = DeriveNormalZ(InXY=ConstantBiasScale(None=ComponentMask(None=@n18) {g=true, r=true}) {bias=-0.5, desc="Bringing to -1, 1 range", scale=2}) {desc=Deriving B from RG}
  n21 = VertexNormalWS()
  LinearInterpolate(A=Add(A=@n19.Result) {constB=0.1}, B=@n19.Result, Alpha=DotProduct(A=@n20, B=@n21))
  LinearInterpolate(
   A=@n20
   B=MaterialFunctionCall(BaseNormal=@n21, AdditionalNormal=@n20) {materialFunction=&BlendAngleCorrectedNormals}.Result
   Alpha=ScalarParameter() {parameterName=FrontEndNormals}
  )
  n22 = MaterialFunctionCall(
   Color_Morning=VectorParameter() {group=TOD, parameterName=TOD_Morning, defaultValue=<rgba 0.3675 0.1266 0.0895 1>}.RGB
   Color_Day=VectorParameter() {group=TOD, parameterName=TOD_Afternoon, defaultValue=<rgba 1 1 1 1>}.RGB
   Color_Evening=VectorParameter() {group=TOD, parameterName=TOD_Evening, defaultValue=<rgba 0.3316 0.1538 0.1088 1>}.RGB
   Color_Night=VectorParameter() {group=TOD, parameterName=TOD_Night, defaultValue=<rgba 0.0874 0.1753 0.1753 1>}.RGB
  ) {materialFunction=&Color_Time_of_Day}
  n23 = CollectionParameter() {
   collection=&FortniteMaterialParameters
   parameterName=SunAndMoonAtmosphereDirectionalVector
  }
  n24 = ComponentMask(None=@n23) {b=true, g=true, r=true}
  n25 = PixelNormalWS()
  n26 = LinearInterpolate(
   A=VectorParameter() {parameterName=emissiveColorAgainstSun, defaultValue=<rgba 0.0677 0.0548 0.0487 1>}.RGB
   B=VectorParameter() {parameterName=EmissiveColorInSun}.RGB
   Alpha=Clamp(None=DotProduct(A=@n24, B=@n25))
  )
  MaterialFunctionCall(Emissive=Multiply(A=Multiply(A=@n22.Result, B=@n26), B=ScalarParameter() {defaultValue=0.05, parameterName=EmissiveIntensity})) {materialFunction=&MF_FX_EmissiveAdjustments}
  n27 = TextureCoordinate()
  n28 = ComponentMask(None=@n27) {r=true}
  n29 = ScalarParameter() {defaultValue=4.1705, parameterName=Power}
  n30 = SphereMask(A=ComponentMask(None=@n27) {g=true}, B=Constant2Vector() {g=0.5, r=0.5}, Radius=Constant() {r=1}, Hardness=Constant() {r=0.5})
  n31 = Saturate(None=Add(A=Power(Base=@n28, Exp=@n29) {constExponent=4}, B=OneMinus(None=@n30)))
  n32 = Constant() {desc=Softness}
  n33 = Add(A=@n32)
  n34 = Subtract(A=Add(A=Add(A=@n6.B, B=@n12.B), B=@n17.B), B=Subtract(A=Multiply(A=@n31, B=@n33), B=@n32))
  Multiply(A=Saturate(None=Divide(A=@n34, B=@n32)), B=ParticleColor().A)
  OneMinus(None=@n28)
  ScalarParameter() {defaultValue=0.74, parameterName=Spec}
  ScalarParameter() {defaultValue=0.376, parameterName=Metal}
 ]
}
