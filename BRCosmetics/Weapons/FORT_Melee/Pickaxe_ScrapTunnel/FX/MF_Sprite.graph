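# Material function in the ScrapTunnel pickaxe FX folder that turns geometry into a camera-facing sprite using World Position Offset.
# UVs, offset by a pivot and scaled to object bounds by default, are laid along rotatable view-space axes around a center location.
# Optional inputs chain extra WPO and shift the sprite along the camera vector.
# A second output re-orients a normal map to the sprite axes in tangent space.
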MaterialFunction /BRCosmetics/Weapons/FORT_Melee/Pickaxe_ScrapTunnel/FX/MF_Sprite
graph {nodes [
  n1 = TextureCoordinate()
  n2 = FunctionInput() {bUsePreviewValueAsDefault=true, description=In UV Space, inputName=Pivot Offset, inputType=FunctionInput_Vector2, sortPriority=2}
  n3 = MaterialFunctionCall() {materialFunction=&ObjectLocalBounds}
  n4 = MaterialFunctionCall() {materialFunction=/Engine/Functions/Engine_MaterialFunctions02/WorldPositionOffset/ObjectScale}
  n5 = FunctionInput(Preview=Multiply(A=ComponentMask(None=@n3.Local Bounds Size) {g=true, r=true}, B=ComponentMask(None=@n4.Scale XYZ) {g=true, r=true})) {
   bUsePreviewValueAsDefault=true
   description=Defaults to object's local bounds * scale
   inputName=XY Scale
   inputType=FunctionInput_Vector2
   sortPriority=1
   previewValue=<111 111 0 1>
  }
  n6 = Multiply(
   A=Subtract(
    A=FunctionInput(Preview=@n1) {bUsePreviewValueAsDefault=true, description=Clamped 0-1, inputName=UV, inputType=FunctionInput_Vector2}
    B=Add(A=@n2) {constB=0.5}
   )
   B=@n5
  )
  n7 = FunctionInput() {bUsePreviewValueAsDefault=true, description=Rotates the sprite, inputName=Rotation, inputType=FunctionInput_Scalar, sortPriority=3}
  n8 = Cosine(None=@n7)
  n9 = Sine(None=@n7)
  n10 = Constant()
  n11 = Transform(None=AppendVector(A=AppendVector(A=@n8, B=@n9), B=@n10)) {transformSourceType=TRANSFORMSOURCE_View}
  n12 = Multiply(A=ComponentMask(None=@n6) {r=true}, B=@n11)
  n13 = Add(A=@n7) {constB=0.25}
  n14 = AppendVector(A=Cosine(None=@n13), B=Sine(None=@n13))
  n15 = Transform(None=Multiply(A=AppendVector(A=@n14, B=@n10)) {constB=-1}) {transformSourceType=TRANSFORMSOURCE_View}
  n16 = Multiply(A=ComponentMask(None=@n6) {g=true}, B=@n15)
  n17 = FunctionInput(Preview=MaterialFunctionCall() {materialFunction=&ComponentPivotLocation}.Result) {
   bUsePreviewValueAsDefault=true
   description=The world position of the center of the sprite
   inputName=Center Location
   sortPriority=4
   previewValue=<0 0 0 0>
  }
  n18 = WorldPosition()
  n19 = FunctionInput() {
   bUsePreviewValueAsDefault=true
   description="Additional WPO, optional, allows you to chain other WPO operations"
   inputName=Additional World Position Offset
   sortPriority=5
  }
  n20 = Add(A=@n18.XYZ, B=@n19)
  n21 = Add(A=Add(A=Add(A=@n12, B=@n16), B=Subtract(A=@n17, B=@n20)), B=Multiply(A=@n19) {constB=2}) {desc=To World Position Offset}
  n22 = Normalize(VectorInput=Subtract(A=CameraPositionWS(), B=Add(A=@n21, B=@n18.XYZ)))
  n23 = FunctionInput() {
   bUsePreviewValueAsDefault=true
   description="Pushes the sprite towards (+) or away (-) from camera, while keeping the screen size stable"
   inputName=Camera Offset
   inputType=FunctionInput_Scalar
   sortPriority=7
  }
  FunctionOutput(None=Add(A=@n21, B=Multiply(A=@n22, B=@n23)))
  n24 = FunctionInput() {
   bUsePreviewValueAsDefault=true
   description="Tangent Space, optional, transforms a Normal Map so it works correctly."
   inputName=Normal
   sortPriority=6
   previewValue=<0 0 1 1>
  }
  n25 = MaterialFunctionCall(Float3=@n24) {materialFunction=&BreakOutFloat3Components}
  n26 = Add(A=Multiply(A=@n11, B=@n25.R), B=Multiply(A=@n15, B=@n25.G))
  n27 = Multiply(A=Transform(None=Constant3Vector() {constant=<rgba 0 0 -1 0>}) {transformSourceType=TRANSFORMSOURCE_View}, B=@n25.B)
  FunctionOutput(None=Transform(None=Add(A=@n26, B=@n27)) {transformSourceType=TRANSFORMSOURCE_World, transformType=TRANSFORM_Tangent}) {description="Transformed Normal Input, in Tangent Space", outputName=Normal, sortPriority=1}
 ]}
