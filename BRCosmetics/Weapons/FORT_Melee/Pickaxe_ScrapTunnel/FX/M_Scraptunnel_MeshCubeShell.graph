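# Masked, two-sided lit Niagara material for the ScrapTunnel pickaxe FX, rendering an animated shell of small cube sprites.
# A time-rotated vector dotted with normalized local position forms a moving mask for sprite shaping and opacity fade.
# MF_Sprite builds sprites around centers offset by normal push and rotation, textured from a 2x2 packed flipbook supplying normals and cutout.
# Two tint colors, time-of-day roughness and normals optionally blended with the sprite's shade them.
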Material /BRCosmetics/Weapons/FORT_Melee/Pickaxe_ScrapTunnel/FX/M_Scraptunnel_MeshCubeShell
graph {
 properties {
  bAutomaticallySetUsageInEditor=false
  bUsedWithNiagaraMeshParticles=true
  bUsedWithNiagaraSprites=true
  blendMode=BLEND_Masked
  translucencyLightingMode=TLM_Surface
  twoSided=true
  usedShadingModels=Default Lit
 }
 nodes [
  LinearInterpolate(
   A=VectorParameter() {defaultValue=<rgba 0.05 0.0481 0.041 1>}.RGB
   B=VectorParameter() {parameterName=Param_1, defaultValue=<rgba 0.01 0.0091 0.0069 1>}.RGB
   Alpha=NamedRerouteUsage()
  )
  n1 = MaterialFunctionCall(
   Value_Morning=ScalarParameter() {defaultValue=0.5, parameterName=Rough_Morning}
   Value_Day=ScalarParameter() {defaultValue=0.45, parameterName=Rough_Day}
   Value_Evening=ScalarParameter() {defaultValue=0.5, parameterName=Rough_Evening}
   Value_Night=ScalarParameter() {defaultValue=0.5, parameterName=Rough_Night}
  ) {materialFunction=&Scalar_Time_of_Day}
  n2 = MaterialFunctionCall!(
   Texture=TextureObject() {texture=&/T_ScrapTunnel_CubesFlip_Packed}
   UVs
   SubImages=Constant2Vector() {g=2, r=2}
   Frame=LinearInterpolate(Alpha=NamedRerouteUsage()) {constB=4}
  ) {materialFunction=/Game/Global/Materials/Character_Parent_2023/FlipbookFace_Parent_2023/MaterialFunctions/Common/MF_FlipbookSubUV}
  n3 = DeriveNormalZ(InXY=@n2.RGB)
  LinearInterpolate(A=Add(A=@n1.Result) {constB=0.1}, B=@n1.Result, Alpha=DotProduct(A=@n3, B=VertexNormalWS()))
  n4 = LinearInterpolate(B=ScalarParameter() {defaultValue=1.5, parameterName=Sprite Size}, Alpha=NamedRerouteUsage())
  n5 = NamedRerouteUsage()
  n6 = NamedRerouteUsage()
  n7 = ScalarParameter() {defaultValue=0.5, parameterName=NormalPushRandom}
  n8 = Multiply(
   A=VertexNormalWS()
   B=Multiply(A=ScalarParameter() {defaultValue=6, parameterName=Normal Push Intensity}, B=Multiply(A=@n6, B=@n7) {constB=0})
  )
  n9 = Power(Base=NamedRerouteUsage(), Exp=ScalarParameter() {defaultValue=4, parameterName=Normal Push Exponent})
  n10 = NamedRerouteUsage()
  n11 = Add(
   A=Multiply(A=NamedRerouteUsage(), B=ScalarParameter() {defaultValue=15.8595, parameterName=Sprite Mask Rotation Amount}) {constA=0.01}
   B=Multiply(A=Time(), B=ScalarParameter() {defaultValue=0.333, parameterName=Rotation Speed})
  )
  n12 = Constant()
  n13 = Normalize(VectorInput=CrossProduct(A=@n10, B=Constant3Vector() {constant=<rgba 0 0 1 0>}))
  n14 = Normalize(VectorInput=Add(A=RotateAboutAxis(NormalizedRotationAxis=@n10, RotationAngle=@n11, PivotPoint=@n12, Position=@n13), B=@n13))
  n15 = Add(A=Add(A=Multiply(A=@n8, B=@n9), B=Transform(None=@n14) {transformSourceType=TRANSFORMSOURCE_Local}), B=WorldPosition().XYZ)
  LinearInterpolate(
   A=@n3
   B=MaterialFunctionCall(
    BaseNormal=MaterialFunctionCall!(UV, XY Scale=@n4, Pivot Offset, Rotation=@n5, Center Location=@n15, Additional World Position Offset, Normal, Camera Offset) {materialFunction=&/MF_Sprite}.Normal
    AdditionalNormal=@n3
   ) {materialFunction=&BlendAngleCorrectedNormals}.Result
   Alpha=ScalarParameter() {parameterName=FrontEndNormals}
  )
  Multiply(A=NamedRerouteUsage(), B=ComponentMask(None=@n2.RGB) {b=true})
  n16 = Time()
  n17 = ScalarParameter() {defaultValue=0.1, parameterName=Time Scale}
  NamedRerouteDeclaration(None=Add(A=Multiply(A=@n16, B=@n17), B=NamedRerouteUsage())) {name=Time Scaled, nodeColor=<rgba 1 0.3281 0 1>}
  n18 = VertexColor()
  NamedRerouteDeclaration(None=@n18.R) {name=VertexColorR, nodeColor=<rgba 0 0.1797 1 1>}
  n19 = LocalPosition()
  n20 = Constant3Vector() {constant=<rgba 0 1 0 0>}
  n21 = Multiply(A=NamedRerouteUsage(), B=ScalarParameter() {defaultValue=0.5, parameterName=Axis Rotation Speed B})
  n22 = Constant()
  n23 = Constant3Vector() {constant=<rgba 0 0 1 0>}
  n24 = Multiply(A=NamedRerouteUsage(), B=ScalarParameter() {defaultValue=1.2, parameterName=Axis Rotation Speed A})
  n25 = Constant()
  n26 = Constant3Vector() {constant=<rgba 1 0 0 0>}
  n27 = Normalize(VectorInput=Add(A=RotateAboutAxis(NormalizedRotationAxis=@n23, RotationAngle=@n24, PivotPoint=@n25, Position=@n26), B=@n26))
  n28 = Normalize(VectorInput=Add(A=RotateAboutAxis(NormalizedRotationAxis=@n20, RotationAngle=@n21, PivotPoint=@n22, Position=@n27), B=@n27))
  n29 = ConstantBiasScale(None=DotProduct(A=Normalize(VectorInput=@n19.XYZ), B=NamedRerouteDeclaration(None=@n28) {name=Rotated Vector, nodeColor=<rgba 1 0 0.5391 1>}))
  n30 = ScalarParameter() {defaultValue=0.5, parameterName=Mask Size}
  n31 = OneMinus(None=Saturate(None=Divide(A=@n29, B=@n30)))
  NamedRerouteDeclaration(None=Power(Base=@n31, Exp=ScalarParameter() {defaultValue=2, parameterName=Mask Exponent})) {name=Sprite Mask, nodeColor=<rgba 0 1 0.6484 1>}
  n32 = Divide(A=@n31, B=ScalarParameter() {defaultValue=0.33, parameterName=Opacity Transition Size})
  NamedRerouteDeclaration(None=VertexInterpolator(VS=Saturate(None=@n32)).PS) {name=Opacity, nodeColor=<rgba 0.6875 1 0 1>}
  NamedRerouteDeclaration(None=@n18.G) {name=VertexColorG, nodeColor=<rgba 0.5 1 0 1>}
  ScalarParameter() {defaultValue=0.75, parameterName=Spec}
  ScalarParameter() {defaultValue=0.375, parameterName=Metal}
 ]
}
