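# Niagara idle effect for the ScrapTunnel pickaxe, combining a hilt mesh, animated ribbon chains and small drifting mesh blocks, all placed through a shared socket reader.
# BlockRibbons partitions one burst into chains around a cylinder, animating width and sway with waveforms and bending them against smoothed owner velocity.
# Blocks (GPU) and lower-rate BlocksLow (CPU) continuously spawn short-lived meshes with curl noise and spring forces tethered to a socket.
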NiagaraSystem /BRCosmetics/Weapons/FORT_Melee/Pickaxe_ScrapTunnel/FX/NS_Scraptunnel_Pickaxe_Idle
dependencies {
 usedDataInterfaces [
  &NiagaraDataInterfaceCurve
  &NiagaraDataInterfaceArrayFloat
  &NiagaraDataInterfaceArrayInt32
  &NiagaraDataInterfaceSocketReader
 ]
 usedDynamicInputs [&SelectIntFromArray]
 usedModules [
  &self:SystemSpawnScript.SystemScriptSource.SystemScriptGraph.NiagaraNodeAssignment_0.SetVariables_#1
  &self:SystemSpawnScript.SystemScriptSource.SystemScriptGraph.NiagaraNodeAssignment_1.SetVariables_#2
  &SystemState
 ]
}
emitters {
 BlockMesh_Hilt {
  emitterUpdateScript [
   &EmitterState(Life Cycle Mode=::NewEnumerator1, Loop Behavior=::NewEnumerator1, Loop Duration Mode=::NewEnumerator1, UseLoopDelay=false)
   &SpawnBurst_Instantaneous(Use Spawn Probability=false, UseLoopCountLimit=false)
   /SetVariables(Emitter.FrontEndNormal:NiagaraFloat=&FN_FrontEndFloatModifier(Gameplay=1, Front End=0)) {isSetParameters=true}
  ]
  particleSpawnScript [
   &InitializeParticle(Color Mode=::NewEnumerator0, Lifetime=5, Mesh Scale Mode=::NewEnumerator3, UsePositionOffset=true, Write Material Random=true)
   &InitialMeshOrientation(Mesh Axis To Orient=<0 0 -1>, Mesh Orientation Mode=::NewEnumerator2, Rotation=<0 0 0.225>, Use Rotation Vector=true)
  ]
  particleUpdateScript [
   &ParticleState(Kill Particles When Lifetime Has Elapsed=false, Let Infinitely Lived Particles Die When Emitter Deactivates=true)
   &SocketLocation(
    SamplePosOffset=true
    Sampled Position Offset=<0 -4.4002 0.575>
    Socket Reader=$System.Socket Reader
    Socket Sampling Mode=::NewEnumerator3
    VelocitySpeedLimit=false
   )
  ]
  rendererClasses [&NiagaraMeshRendererProperties]
  renderers [{rendererClass=&NiagaraMeshRendererProperties, rendererIndex=0}]
 }
 BlockRibbons {
  emitterSpawnScript [&FN_RibbonIDOffsetUtility()]
  emitterUpdateScript [
   &EmitterState(Life Cycle Mode=::NewEnumerator1, Loop Behavior=::NewEnumerator1, Loop Duration Mode=::NewEnumerator1, UseLoopDelay=false)
   &SpawnBurst_Instantaneous(
    Spawn Count=&Multiply_Int(A=$System.ChainNumber, B=$System.ParticlesPerChain)
    Use Spawn Probability=false
    UseLoopCountLimit=false
   )
   /SetVariables(Emitter.FrontEndNormal:NiagaraFloat=&FN_FrontEndFloatModifier(Gameplay=1, Front End=0)) {isSetParameters=true}
  ]
  particleSpawnScript [
   &PartitionParticles(
    Number Of Partitions=$System.ChainNumber
    Particles Per Partition:NiagaraInt32=$System.ParticlesPerChain
    Partition Mode:ENiagara_PartitionMode_v2=::NewEnumerator1
    Write Partition Index to Ribbon Link Order:ENiagara_AttributeSamplingApplyOutput=::NewEnumerator0
    Write Partition to Ribbon ID=::NewEnumerator0
   )
   &InitializeParticle(
    Color=<rgba 0.5677 0.4966 0.4228 1>
    Ribbon Width=2.5
    Ribbon Width Mode=::NewEnumerator1
    Sprite Size Mode=::NewEnumerator3
    Uniform Sprite Size=3
    UsePositionOffset=false
    Write Material Random=true
   )
   &SocketLocation(
    Orientation Sampling=::NewEnumerator1
    SamplePosOffset=true
    Sampled Position Offset=<0 5 0>
    Socket Reader=$System.Socket Reader
    Socket Sampling Mode=::NewEnumerator3
    VelocitySpeedLimit=false
   )
   &ShapeLocation(
    Additional Quaternion Rotation=true
    Additional Yaw / Pitch / Roll=<0 0 0.25>
    Cylinder Distribution=::NewEnumerator2
    Cylinder Height=1
    Cylinder Radius=4.5
    MaskSpawnGroup=false
    Rotation Mode=::NewEnumerator3
    Rotation Quaternion=$Output.SocketLocation.SampledRotation
    Shape Primitive=::NewEnumerator1
    U Position=$Particles.PartitionParticles.NormalizedPartition
    UseOwnerScale=true
   )
   /SetVariables(Particles.RibbonLinkOrder:NiagaraFloat=$Particles.PartitionParticles.Normalized Index In Partition) {isSetParameters=true}
   /SetVariables(Particles.Position=&AddVectorToPosition(
     Position=$Particles.ShapeLocation.ShapePosition
     Vector=&TransformVector(
      Vector=&MakeVector(
       X=&Multiply_FloatByInt(
        Float=&RandomRangeFloat(Minimum=6, Maximum=7)
        Integer=$Particles.PartitionParticles.Index In Partition
       )
       Y=&FloatFromCurve(
        FloatCurve=DataInterface() {propertyValues {Curve {defaultValue=1, keys [<at 0 0 RCIM_Cubic>,<at 1 1 RCIM_Cubic>]}}}
        CurveIndex=$Particles.PartitionParticles.Normalized Index In Partition
        Scale Curve=3
       )
       Z=0
      )
      Destination Space=::Simulation
     )
    )) {isSetParameters=true}
   &DynamicMaterialParameters(
    Index 0 Param 1=&SelectFloatFromArray(
     Direct Array Index=$Particles.PartitionParticles.Partition
     Float Selection Array=DataInterface() {propertyValues {FloatData [0.5,0.1,0.9,0.2,0.7,0.3,0.6,0.8,0.4,0,1,0.5,0.2,0.9,0.1,0.7]}}
    )
    Index 0 Param 2=&SelectFloatFromArray(
     Direct Array Index=$Particles.PartitionParticles.Partition
     Float Selection Array=DataInterface() {propertyValues {FloatData [0.93,0.81,0.99,0.87,0.95,0.84,0.91,0.98,0.89,1,0.8,0.96,0.83,0.97,0.86,0.92]}}
    )
    Module.Index0Float0Write=true
    Module.Index0Float1Write=true
    Module.Index0Float2Write=true
    Module.Index0Float3Write=true
   ) {module=DynamicMaterialParameters001}
   /SetVariables(Particles.FanPos:NiagaraPosition=&AddVectorToPosition(
     Position=$Particles.Position
     Vector=&Multiply_Vector(
      A=&RotateVectorByQuaternion(Vector To Rotate=$Particles.ShapeLocation.ShapeNormal, Quaternion=<0 0 0 1>)
      B=&MakeVector(X=0, Y=0, Z=0)
     )
    )) {isSetParameters=true}
  ]
  particleUpdateScript [
   &ParticleState(Kill Particles When Lifetime Has Elapsed=false, Let Infinitely Lived Particles Die When Emitter Deactivates=true)
   &ScaleRibbonWidth(Ribbon Width Scale=&Multiply_Float(
     A=&Waveform(
      Global Input Position=&RandomRangeFloat(Minimum=0, Maximum=1)
      Global Phase=&Add_Float(A=$Emitter.Age, B=$Particles.Initial.PartitionParticles.Normalized Index In Partition)
      "Waveform [1]"=::NewEnumerator0
      "[1] Frequency"=0.25
      "[1] Amplitude Min/Max"=<0.5 2>
      "[1] Bias"=1
     )
     B=&FloatFromCurve(
      FloatCurve=DataInterface() {propertyValues {Curve {
         defaultValue=1
         keys [
          {arriveTangent=1.9453, interpMode=RCIM_Cubic, leaveTangent=1.9453, tangentMode=RCTM_User, time=0, value=0.7}
          {arriveTangent=-0.025, interpMode=RCIM_Cubic, leaveTangent=-0.025, time=0.25, value=1}
          {arriveTangent=-0.3355, interpMode=RCIM_Cubic, leaveTangent=-0.3355, tangentMode=RCTM_User, time=1, value=0.675}
         ]
        }}}
      CurveIndex=$Particles.PartitionParticles.Normalized Index In Partition
     )
    ))
   /SetVariables(Particles.Position=&Add_Vector(
     A=Linked() {linkedVariable {name=Particles.FanPos, type=NiagaraPosition}}
     B=&Multiply_VectorByFloat(
      Vector=$Particles.ShapeLocation.ShapeNormal
      Float=&Multiply_Float(
       A=&Waveform(
        Global Input Position=&RandomRangeFloat(Minimum=0, Maximum=1)
        Global Amplitude Scale=&RandomRangeFloat(Minimum=1, Maximum=3.5)
        "Waveform [1]"=::NewEnumerator0
        "[1] Frequency"=&Multiply_Float(A=&RandomRangeFloat(Minimum=0.3, Maximum=0.4), B=1)
        "[1] Amplitude Min/Max"=<-1 1>
       )
       B=&FloatFromCurve(
        FloatCurve=DataInterface() {propertyValues {Curve {
           defaultValue=1
           keys [{arriveTangent=2.0199, interpMode=RCIM_Cubic, leaveTangent=2.0199, tangentMode=RCTM_User, time=0, value=0.2}, <at 1 1 RCIM_Cubic>]
          }}}
        CurveIndex=$Particles.PartitionParticles.Normalized Index In Partition
       )
      )
     )
    )) {isSetParameters=true}
   &InterpolateOverTime(
    Rate Of Change=3
    Target Vector Value:Vector3f=&MakeCustomVectorFromBool(
     Bool=&SetBoolByFloatComparison(A=1000, B=&VectorLength(Vector=$Engine.Owner.Velocity), Comparison Type=::NewEnumerator6)
     True Vector=<1 1 1>
     False Vector=$Engine.Owner.Velocity
    )
    Variable Type=::NewEnumerator2
   )
   /SetVariables(Particles.Position=&AddVectorToPosition(
     Position=$Particles.Position
     Vector=&Multiply_Vector(
      A=&TransformVector(Vector=$Particles.InterpolateOverTime.InterpolateOverTime_V3_.Moving Average, Source Space=::World, Destination Space=::Local)
      B=&Multiply_VectorByFloat(
       Vector=<-1 -1 -1>
       Float=&FloatFromCurve(
        FloatCurve=DataInterface() {propertyValues {Curve {
           defaultValue=0
           keys [{interpMode=RCIM_Cubic, tangentMode=RCTM_User, time=0, value=0}, {interpMode=RCIM_Cubic, tangentMode=RCTM_User, time=1, value=1}]
          }}}
        CurveIndex=$Particles.PartitionParticles.Normalized Index In Partition
        Scale Curve=0.025
       )
      )
     )
    )) {isSetParameters=true}
  ]
  rendererClasses [&NiagaraRibbonRendererProperties]
  renderers [{rendererClass=&NiagaraRibbonRendererProperties}]
 }
 Blocks {
  simTarget=GPUComputeSim
  emitterUpdateScript [
   &EmitterState()
   &SpawnRate(Spawn Probability=0.44, SpawnRate=16, Use Spawn Probability=true)
  ]
  particleSpawnScript [
   &InitializeParticle(
    Color Maximum=<rgba 0.0521 0.0368 0.0255 1>
    Color Minimum=<rgba 0.0417 0.0327 0.0219 1>
    Color Mode=::NewEnumerator2
    Lifetime Max=1.5
    Lifetime Mode=::NewEnumerator1
    Mesh Scale Mode=::NewEnumerator1
    Mesh Uniform Scale Max=0.035
    Mesh Uniform Scale Min=0.02
    UsePositionOffset=false
    Write Material Random=true
   )
   &InitialMeshOrientation(Use Rotation Vector=false)
   &SocketLocation(SamplePosOffset=false, Socket Index=1, Socket Reader=$System.Socket Reader, Socket Sampling Mode=::NewEnumerator2, VelocitySpeedLimit=false) {module=SocketLocation001}
   &ShapeLocation(
    Cylinder Height=2
    Cylinder Radius=8
    MaskSpawnGroup=false
    Offset=<3 0 0>
    Offset Coordinate Space=::Local
    Offset Mode=::NewEnumerator0
    Override Seed=false
    Rotation Mode=::NewEnumerator1
    Shape Origin=$Output.SocketLocation001.SampledPosition
    Shape Primitive=::NewEnumerator1
    Surface Only=true
    Use Endcaps In Surface Only Mode=false
    UseOwnerScale=true
   )
   &AddVelocity(
    Module.Use Falloff Distance=false
    Override Seed=false
    Velocity Mode=::NewEnumerator1
    Velocity Origin=$Output.SocketLocation001.SampledPosition
    Velocity Speed=&RandomRangeFloat(Minimum=10, Maximum=5)
   )
  ]
  particleUpdateScript [
   &ParticleState()
   &SocketLocation(
    Position Sampling=::NewEnumerator1
    SamplePosOffset=true
    Sampled Position Offset=$Particles.ShapeLocation.ShapeVector
    Sampled Position Offset Coordinate Space:ENiagaraSkelMeshTransforms=::NewEnumerator2
    Socket Index=1
    Socket Reader=$System.Socket Reader
    Socket Sampling Mode=::NewEnumerator2
    VelocitySpeedLimit=false
   )
   &CurlNoiseForce(
    Bias Noise=false
    Noise Frequency=25
    Noise Strength=&RandomRangeFloat(Minimum=22, Maximum=33)
    Pan Noise=true
    Pan Noise Field=<0 0 1>
   )
   &ScaleMeshSize(Scale Factor=&Multiply_VectorByFloat(
     Vector=<1 1 1>
     Float=&FloatFromCurve(FloatCurve=DataInterface() {propertyValues {Curve {
         defaultValue=1
         keys [
          {interpMode=RCIM_Cubic, tangentMode=RCTM_User, time=0, value=0}
          {arriveTangent=-0.0885, interpMode=RCIM_Cubic, leaveTangent=-0.0885, tangentMode=RCTM_User, time=0.15, value=1}
          {interpMode=RCIM_Cubic, tangentMode=RCTM_User, time=0.5, value=1}
          {tangentMode=RCTM_User, time=1, value=0}
         ]
        }}})
    ))
   &MeshRotationForce(Rotation=<2 2 2>)
   &Drag(Rotational Drag=0.5, Use Linear Drag=true, Use Rotational Drag=true)
   /Niagara/Modules/Update/Forces/SpringForce(
    " Particle Equilibrium Position":NiagaraPosition=$Output.SocketLocation.SampledPosition
    Attachment Point Velocity:Vector3f=$Engine.Owner.Velocity
    Dampening Coefficient=2
    Desired Minimum Separation Distance=1
    Force Strength:NiagaraFloat=&RandomRangeFloat(Minimum=5, Maximum=10)
    Spring Tightness:NiagaraFloat=&RandomRangeFloat(Minimum=1, Maximum=1)
   )
   &SolveForcesAndVelocity(Clamp Velocity=false, Limit Acceleration=false)
  ]
  rendererClasses [&NiagaraMeshRendererProperties]
  renderers [{rendererClass=&NiagaraMeshRendererProperties, rendererIndex=0}]
 }
 BlocksLow {
  emitterUpdateScript [
   &EmitterState()
   &SpawnRate(Spawn Probability=0.44, SpawnRate=10, Use Spawn Probability=true)
  ]
  particleSpawnScript [
   &InitializeParticle(
    Color Maximum=<rgba 0.0521 0.0368 0.0255 1>
    Color Minimum=<rgba 0.0417 0.0327 0.0219 1>
    Color Mode=::NewEnumerator2
    Lifetime Max=1.5
    Lifetime Mode=::NewEnumerator1
    Mesh Scale Mode=::NewEnumerator1
    Mesh Uniform Scale Max=0.035
    Mesh Uniform Scale Min=0.02
    UsePositionOffset=false
    Write Material Random=true
   )
   &InitialMeshOrientation(Use Rotation Vector=false)
   &SocketLocation(SamplePosOffset=false, Socket Index=1, Socket Reader=$System.Socket Reader, Socket Sampling Mode=::NewEnumerator2, VelocitySpeedLimit=false) {module=SocketLocation001}
   &ShapeLocation(
    Cylinder Height=2
    Cylinder Radius=8
    MaskSpawnGroup=false
    Offset=<3 0 0>
    Offset Coordinate Space=::Local
    Offset Mode=::NewEnumerator0
    Override Seed=false
    Rotation Mode=::NewEnumerator1
    Shape Origin=$Output.SocketLocation001.SampledPosition
    Shape Primitive=::NewEnumerator1
    Surface Only=true
    Use Endcaps In Surface Only Mode=false
    UseOwnerScale=true
   )
   &AddVelocity(
    Module.Use Falloff Distance=false
    Override Seed=false
    Velocity Mode=::NewEnumerator1
    Velocity Origin=$Output.SocketLocation001.SampledPosition
    Velocity Speed=&RandomRangeFloat(Minimum=10, Maximum=5)
   )
  ]
  particleUpdateScript [
   &ParticleState()
   &SocketLocation(
    Position Sampling=::NewEnumerator1
    SamplePosOffset=true
    Sampled Position Offset=$Particles.ShapeLocation.ShapeVector
    Sampled Position Offset Coordinate Space:ENiagaraSkelMeshTransforms=::NewEnumerator2
    Socket Index=1
    Socket Reader=$System.Socket Reader
    Socket Sampling Mode=::NewEnumerator2
    VelocitySpeedLimit=false
   )
   &CurlNoiseForce(
    Bias Noise=false
    Noise Frequency=25
    Noise Strength=&RandomRangeFloat(Minimum=22, Maximum=33)
    Pan Noise=true
    Pan Noise Field=<0 0 1>
   )
   &ScaleMeshSize(Scale Factor=&Multiply_VectorByFloat(
     Vector=<1 1 1>
     Float=&FloatFromCurve(FloatCurve=DataInterface() {propertyValues {Curve {
         defaultValue=1
         keys [
          {interpMode=RCIM_Cubic, tangentMode=RCTM_User, time=0, value=0}
          {arriveTangent=-0.0885, interpMode=RCIM_Cubic, leaveTangent=-0.0885, tangentMode=RCTM_User, time=0.15, value=1}
          {interpMode=RCIM_Cubic, tangentMode=RCTM_User, time=0.5, value=1}
          {tangentMode=RCTM_User, time=1, value=0}
         ]
        }}})
    ))
   &MeshRotationForce(Rotation=<2 2 2>)
   &Drag(Rotational Drag=0.5, Use Linear Drag=true, Use Rotational Drag=true)
   &SpringForce(
    " Particle Equilibrium Position":NiagaraPosition=$Output.SocketLocation.SampledPosition
    Attachment Point Velocity:Vector3f=$Engine.Owner.Velocity
    Dampening Coefficient=2
    Desired Minimum Separation Distance=1
    Force Strength:NiagaraFloat=&RandomRangeFloat(Minimum=5, Maximum=10)
    Spring Tightness:NiagaraFloat=&RandomRangeFloat(Minimum=1, Maximum=1)
   )
   &SolveForcesAndVelocity(Clamp Velocity=false, Limit Acceleration=false)
  ]
  rendererClasses [&NiagaraMeshRendererProperties]
  renderers [{rendererClass=&NiagaraMeshRendererProperties, rendererIndex=0}]
 }
}
summary {
 systemName=NS_Scraptunnel_Pickaxe_Idle
 emitters [
  {emitterName=Blocks, simTarget=GPUComputeSim, rendererClasses [&NiagaraMeshRendererProperties]}
  {emitterName=BlockMesh_Hilt, rendererClasses [&NiagaraMeshRendererProperties]}
  {emitterName=BlockRibbons, rendererClasses [&NiagaraRibbonRendererProperties]}
  {emitterName=BlocksLow, rendererClasses [&NiagaraMeshRendererProperties]}
 ]
 userVariables [
  User.bIsFrontEnd {type=NiagaraBool}
  User.bIsFrontEndPreview {type=NiagaraBool}
 ]
}
