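# Translucent parent material for the ScrapTunnel pickaxe, faking tunnel depth with curved parallax occlusion mapping.
# Panned UVs and a time-driven sine through an FX mask modulate the height fed to MF_POM_Curvature, with dithered step counts.
# Parallax results drive textures, pixel depth offset and a vertex-color-trimmed opacity mask, alongside normal-based vertex offset.
# A Use POM switch selects this or plain character PreFX attributes before the shared PostFX function.
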Material /BRCosmetics/Weapons/FORT_Melee/Pickaxe_ScrapTunnel/Materials/M_Pickaxe_ScrapTunnel_Parent
graph {
 properties {
  allowTranslucentCustomDepthWrites=true
  bHasPixelAnimation=true
  bOutputTranslucentVelocity=true
  bUseMaterialAttributes=true
  bUsedWithMeshDeformer=true
  bUsedWithMorphTargets=true
  bUsedWithSkeletalMesh=true
  blendMode=BLEND_Translucent
  pixelDepthOffsetMode=PDOM_Legacy
  shadingModel=MSM_FromMaterialExpression
  translucencyLightingMode=TLM_SurfacePerPixelLighting
  usedShadingModels=Default Lit | Subsurface | Clear Coat | Subsurface Profile | Cloth
 }
 nodes [
  ComponentMask(None=TextureCoordinate()) {g=true}
  n1 = VertexNormalWS()
  ComponentMask(None=@n1) {b=true}
  n2 = Multiply(A=Time(), B=ScalarParameter() {defaultValue=0.5, parameterName=Indulation Speed})
  n3 = VectorParameter() {group=Scrap Tunnel, parameterName=Parallax UV - Scale, defaultValue=<rgba 2 2 0 0>}
  n4 = AppendVector(A=@n3.R, B=@n3.R)
  n5 = Time()
  n6 = ScalarParameter() {group=Scrap Tunnel, parameterName=Panning Speed}
  n7 = Add(
   A=Multiply(A=ComponentMask(None=@n4) {g=true, r=true}, B=TextureCoordinate())
   B=MaterialFunctionCall(X=Multiply(A=@n5, B=@n6), Y=Constant()) {materialFunction=&MakeFloat2}.Result
  ) {constB=0}
  n8 = TextureSample(UVs=@n7) {samplerType=SAMPLERTYPE_Masks, texture=/BRCosmetics/Weapons/FORT_Melee/Pickaxe_ScrapTunnel/Textures/T_ScrapTunnel_Tilable_01_FX}
  Divide(A=@n2, B=@n8.G)
  n9 = Normalize(VectorInput=ObjectPositionWS())
  LinearInterpolate(B=Multiply(A=@n9, B=ScalarParameter() {defaultValue=-0.18, parameterName=Bevel Collapse}), Alpha=NamedRerouteUsage())
  n10 = MaterialFunctionCall() {materialFunction=&MF_CharacterParent_2023_PreFX}
  n11 = GetMaterialAttributes(None=@n10.Result) {attributeGetTypes [#1]}
  n12 = TextureObjectParameter() {
   group=Scrap Tunnel
   parameterName=Height
   texture=/BRCosmetics/Weapons/FORT_Melee/Pickaxe_ScrapTunnel/Textures/Protoype/T_Grid_Height_Gapless
  }
  n13 = Multiply(A=ScalarParameter() {defaultValue=0.16, group=Scrap Tunnel, parameterName=Main Height Ratio}, B=@n4)
  n14 = MaterialFunctionCall!(Sine Phase=Subtract(A=@n8.G, B=@n2), Value 1=ScalarParameter() {parameterName=Min Modulation}, Value 2) {materialFunction=/Engine/Functions/Engine_MaterialFunctions02/Utility/Sine_Remapped}
  n15 = Multiply(A=Saturate(None=ComponentMask(None=@n14.Remapped Sine) {r=true}), B=@n8.B)
  n16 = ScalarParameter() {defaultValue=1, group=Scrap Tunnel, parameterName=m}
  n17 = MaterialFunctionCall() {materialFunction=&DitherTemporalAA}
  n18 = VectorParameter() {parameterName=Curvature Axis, defaultValue=<rgba -1 0 0 0>}
  n19 = MaterialFunctionCall!(
   Heightmap Texture=@n12
   Height Ratio=LinearInterpolate(A=@n13, B=Multiply(A=@n15, B=@n13), Alpha=@n8.B)
   Min Steps=ScalarParameter() {defaultValue=8, parameterName=MinSteps}
   Max Steps=Multiply(
    A=ScalarParameter() {defaultValue=128, group=Scrap Tunnel, parameterName=MaxSteps}
    B=LinearInterpolate(B=@n16, Alpha=@n17.Result) {constA=1}
   )
   UVs=@n7
   Heightmap Channel=Constant4Vector() {constant=<rgba 1 0 0 0>}.RGBA
   -----------------------------
   Use World Coords
   Specify Manual Texture Size
   Manual Texture Size
   ------------------------------
   "Render Shadows (Occlusion Mapping)"=StaticBoolParameter() {group=Scrap Tunnel, parameterName="Render Shadows (Occlusion Mapping) (SB)"}
   Light Vector=AtmosphericLightVector()
   Shadow Steps
   Shadow Penumbra
   -------------------------------
   Curvature Axis=MaterialFunctionCall(X=@n18.R, Y=@n18.R, Z=@n18.R) {materialFunction=&MakeFloat3}.Result
   Curvature Radius=ScalarParameter() {defaultValue=1.2908, parameterName=Curvature}
   Step Multiplier
  ) {materialFunction=&/MF_POM_Curvature}
  n20 = ComponentMask(None=@n19.Parallax UVs) {g=true}
  n21 = Add(A=OneMinus(None=@n20), B=ScalarParameter() {defaultValue=0.44, group=Scrap Tunnel, parameterName=Mask Cut - Bottom})
  n22 = Subtract(A=@n20, B=ScalarParameter() {defaultValue=0.65, group=Scrap Tunnel, parameterName=Mask Cut})
  n23 = NamedRerouteUsage()
  n24 = Multiply(A=Ceil(None=Multiply(A=@n21, B=@n22)), B=@n23)
  n25 = VertexColor()
  n26 = ScalarParameter() {defaultValue=-1.0027, group=Scrap Tunnel, parameterName=Vertex Offset}
  n27 = Multiply(A=@n1, B=@n26)
  n28 = NamedRerouteUsage()
  n29 = TextureSampleParameter2D(UVs=@n28) {
   parameterName=Roughness
   samplerType=SAMPLERTYPE_LinearGrayscale
   texture=/BRCosmetics/Weapons/FORT_Melee/Pickaxe_ScrapTunnel/Textures/T_ScrapTunnel_Tilable_01_R
  }
  n30 = NamedRerouteUsage()
  n31 = ScalarParameter() {defaultValue=1, group=Scrap Tunnel, parameterName=PDO Intensity}
  n32 = SetMaterialAttributes!(
   MaterialAttributes=@n11.MaterialAttributes
   Opacity Mask=ShadowReplace(Default=Subtract(A=@n24, B=@n25.R), Shadow=Constant())
   World Position Offset=NamedRerouteDeclaration(None=@n27) {name=Vertex Offset, nodeColor=<rgba 0 0.6953 1 1>}
   Normal=NamedRerouteUsage()
   Metallic=ScalarParameter() {group=Scrap Tunnel, parameterName=Metallic}
   Specular=ScalarParameter() {defaultValue=0.75, parameterName=Specular}
   Roughness=@n29.RGB
   Base Color=Add(
    A=@n11.Base Color
    B=TextureSample(UVs=@n28) {texture=/BRCosmetics/Weapons/FORT_Melee/Pickaxe_ScrapTunnel/Textures/T_ScrapTunnel_Pickaxe_Tilable_01_D_PNG}.RGB
   )
   Pixel Depth Offset=StaticSwitchParameter(True=Multiply(A=@n30, B=@n31), False=Constant() {r=0}) {defaultValue=true, desc=pdo, parameterName=Use PDO}
  ) {attributeSetTypes [
    #2
    #3
    #4
    #5
    #6
    #7
    #1
    #8
   ]}
  MaterialFunctionCall(
   Material=StaticSwitchParameter(True=@n32, False=@n10.Result) {parameterName=Use POM}
   M=@n10.M
   Pre Skinned Local Position=@n10.M
   Distance Blend=@n10.M
   Fresnel=@n10.M
  ) {materialFunction=&MF_CharacterParent_2023_PostFX}
  n33 = Constant()
  MaterialFunctionCall(R=@n33, G=@n33, B=@n27) {materialFunction=&AppendMany}
  Multiply(A=@n23, B=VertexColor().G)
  n34 = VertexColor()
  Multiply(A=@n26, B=@n34.B)
  NamedRerouteDeclaration(None=@n19.Parallax UVs) {name=Parallax UVs, nodeColor=<rgba 0.0547 1 0 1>}
  n35 = TextureSampleParameter2D!(UVs=NamedRerouteUsage(), "DDX(UVs)"=DDX(Value=@n7), "DDY(UVs)"=DDY(Value=@n7), Apply View MipBias) {
   group=Scrap Tunnel
   mipValueMode=TMVM_Derivative
   parameterName=Normal
   samplerType=SAMPLERTYPE_Normal
   texture=/BRCosmetics/Weapons/FORT_Melee/Pickaxe_ScrapTunnel/Textures/Protoype/T_Grid_N_Gapless
  }
  NamedRerouteDeclaration(None=@n35.RGB) {name=NRM, nodeColor=<rgba 1 0 0.375 1>}
  NamedRerouteDeclaration(None=@n19.Pixel Depth Offset) {name=PDO, nodeColor=<rgba 1 0.9844 0 1>}
  NamedRerouteDeclaration(None=@n19.Silhouette Mask) {name=Silhouette Mask, nodeColor=<rgba 0.6641 0 1 1>}
  NamedRerouteDeclaration(None=@n19.Shadow) {name=Shadow, nodeColor=<rgba 0 1 0.0156 1>}
  NamedRerouteDeclaration(None=@n34.G) {name=Outer, nodeColor=<rgba 1 0.6328 0 1>}
  NamedRerouteDeclaration(None=@n34.B) {name=Inner, nodeColor=<rgba 0.3125 0 1 1>}
  NamedRerouteUsage()
  NamedRerouteUsage()
  NamedRerouteUsage()
  OneMinus(None=@n9)
  OneMinus(None=@n8.B)
  OneMinus(None=@n7)
  ScalarParameter() {defaultValue=1, parameterName=Spec}
  ScalarParameter() {defaultValue=0.5, desc=r, group=Scrap Tunnel, parameterName=Roughness}
  StaticBool()
  TextureSample() {texture=&T_Grid_Height_Gapless}
  VectorParameter() {group=Scrap Tunnel, parameterName=Color, defaultValue=<rgba 0.5 0.5 0.5 0>}
  VertexNormalWS()
  WorldPosition()
 ]
}
